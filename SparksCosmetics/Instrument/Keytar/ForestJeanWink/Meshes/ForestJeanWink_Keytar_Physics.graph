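# Physics asset for the ForestJeanWink keytar cosmetic, defining capsule bodies and joints for a tassel and a ball on a string that hang from the handle.
# Seven four-segment tassel strands (A to G) branch from Tassle_root, with swing limits generally narrowing toward the tips.
# A six-segment string chain, fully locked in rotation at the handle, ends in a zero-length capsule acting as the ball.
# Every joint locks twist.
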PhysicsAsset /SparksCosmetics/Instrument/Keytar/ForestJeanWink/Meshes/ForestJeanWink_Keytar_Physics
bodies {
 Tassle_root [{shapeType=Capsule, length=0.505, radius=0.505}]
 dyn_ball [{shapeType=Capsule, length=0, radius=3.5, shapeName=dyn_ball_capsule, center=<3.25 0 0>}]
 dyn_string_01 [{shapeType=Capsule, length=0.505, radius=0.505, shapeName=dyn_string_01_capsule, rotation=<rot 0 -140.1668 -85.1469>}]
 dyn_string_02 [{shapeType=Capsule, length=0.6622, radius=0.505, shapeName=dyn_string_02_capsule, center=<0.7 0 0>, rotation=<rot 90 0.006 0>}]
 dyn_string_03 [{shapeType=Capsule, length=0.5811, radius=0.505, shapeName=dyn_string_03_capsule, center=<0.6 0 0>, rotation=<rot 0 90.0961 -89.9616>}]
 dyn_string_04 [{shapeType=Capsule, length=0.5, radius=0.505, shapeName=dyn_string_04_capsule, center=<0.6 0 0>, rotation=<rot 0 90.0813 -89.9446>}]
 dyn_string_05 [{shapeType=Capsule, length=0.7874, radius=0.505, shapeName=dyn_string_05_capsule, center=<0.8 0 0>, rotation=<rot 0 90.053 -89.956>}]
 dyn_string_06 [{shapeType=Capsule, length=0.505, radius=0.505, shapeName=dyn_string_06_capsule, center=<0.6 0 0>, rotation=<rot 0 90.0052 -89.974>}]
 dyn_tassle_A_01 [{shapeType=Capsule, length=1, radius=0.505, shapeName=dyn_tassle_A_01_capsule, center=<0.75 0 0>, rotation=<rot 90 0 0>}]
 dyn_tassle_A_02 [{shapeType=Capsule, length=1.3, radius=0.505, shapeName=dyn_tassle_A_02_capsule, center=<0.85 0 0>, rotation=<rot 90 0 0>}]
 dyn_tassle_A_03 [{shapeType=Capsule, length=1.2252, radius=0.505, shapeName=dyn_tassle_A_03_capsule, center=<0.8 0 0>, rotation=<rot 90 0 0>}]
 dyn_tassle_A_04 [{shapeType=Capsule, length=0.505, radius=0.505, shapeName=dyn_tassle_A_04_capsule, center=<0.5 0 0>, rotation=<rot 90 0 0>}]
 dyn_tassle_B_01 [{shapeType=Capsule, length=1, radius=0.505, shapeName=dyn_tassle_B_01_capsule, center=<0.75 0 0>, rotation=<rot 90 0 0>}]
 dyn_tassle_B_02 [{shapeType=Capsule, length=1.25, radius=0.505, shapeName=dyn_tassle_B_02_capsule, center=<0.85 0 0>, rotation=<rot 90 0 0>}]
 dyn_tassle_B_03 [{shapeType=Capsule, length=1.2252, radius=0.505, shapeName=dyn_tassle_B_03_capsule, center=<0.8 0 0>, rotation=<rot 90 0 0>}]
 dyn_tassle_B_04 [{shapeType=Capsule, length=0.505, radius=0.505, shapeName=dyn_tassle_B_04_capsule, center=<0.5 0 0>, rotation=<rot 90 0 0>}]
 dyn_tassle_C_01 [{shapeType=Capsule, length=1, radius=0.505, shapeName=dyn_tassle_C_01_capsule, center=<0.75 0 0>, rotation=<rot 90 0 0>}]
 dyn_tassle_C_02 [{shapeType=Capsule, length=1.25, radius=0.505, shapeName=dyn_tassle_C_02_capsule, center=<0.8 0 0>, rotation=<rot 90 0 0>}]
 dyn_tassle_C_03 [{shapeType=Capsule, length=0.8, radius=0.505, shapeName=dyn_tassle_C_03_capsule, center=<0.5 0 0>, rotation=<rot 90 0 0>}]
 dyn_tassle_C_04 [{shapeType=Capsule, length=0.505, radius=0.505, shapeName=dyn_tassle_C_04_capsule, center=<0.5 0 0>, rotation=<rot 90 0 0>}]
 dyn_tassle_D_01 [{shapeType=Capsule, length=1, radius=0.505, shapeName=dyn_tassle_D_01_capsule, center=<0.75 0 0>, rotation=<rot 90 0 0>}]
 dyn_tassle_D_02 [{shapeType=Capsule, length=0.8, radius=0.505, shapeName=dyn_tassle_D_02_capsule, center=<0.6 0 0>, rotation=<rot 90 0 0>}]
 dyn_tassle_D_03 [{shapeType=Capsule, length=0.7, radius=0.505, shapeName=dyn_tassle_D_03_capsule, center=<0.6 0 0>, rotation=<rot 90 0 0>}]
 dyn_tassle_D_04 [{shapeType=Capsule, length=0.505, radius=0.505, shapeName=dyn_tassle_D_04_capsule, center=<0.5 0 0>, rotation=<rot 90 0 0>}]
 dyn_tassle_E_01 [{shapeType=Capsule, length=0.505, radius=0.505, shapeName=dyn_tassle_E_01_capsule, center=<0.5 0 0>, rotation=<rot 90 0 0>}]
 dyn_tassle_E_02 [{shapeType=Capsule, length=0.8, radius=0.505, shapeName=dyn_tassle_E_02_capsule, center=<0.6 0 0>, rotation=<rot 90 0 0>}]
 dyn_tassle_E_03 [{shapeType=Capsule, length=0.8, radius=0.505, shapeName=dyn_tassle_E_03_capsule, center=<0.7 0 0>, rotation=<rot 90 0 0>}]
 dyn_tassle_E_04 [{shapeType=Capsule, length=0.505, radius=0.505, shapeName=dyn_tassle_E_04_capsule, center=<0.5 0 0>, rotation=<rot 90 0 0>}]
 dyn_tassle_F_01 [{shapeType=Capsule, length=0.505, radius=0.505, shapeName=dyn_tassle_F_01_capsule, center=<0.5 0 0>, rotation=<rot 90 0 0>}]
 dyn_tassle_F_02 [{shapeType=Capsule, length=0.8, radius=0.505, shapeName=dyn_tassle_F_02_capsule, center=<0.7 0 0>, rotation=<rot 90 0 0>}]
 dyn_tassle_F_03 [{shapeType=Capsule, length=0.505, radius=0.505, shapeName=dyn_tassle_F_03_capsule, center=<0.5 0 0>, rotation=<rot 90 0 0>}]
 dyn_tassle_F_04 [{shapeType=Capsule, length=0.505, radius=0.505, shapeName=dyn_tassle_F_04_capsule, center=<0.5 0 0>, rotation=<rot 90 0 0>}]
 dyn_tassle_G_01 [{shapeType=Capsule, length=0.505, radius=0.505, shapeName=dyn_tassle_G_01_capsule, center=<0.5 0 0>, rotation=<rot 90 0 0>}]
 dyn_tassle_G_02 [{shapeType=Capsule, length=0.505, radius=0.505, shapeName=dyn_tassle_G_02_capsule, center=<0.5 0 0>, rotation=<rot 90 0 0>}]
 dyn_tassle_G_03 [{shapeType=Capsule, length=0.505, radius=0.505, shapeName=dyn_tassle_G_03_capsule, center=<0.5 0 0>, rotation=<rot 90 0 0>}]
 dyn_tassle_G_04 [{shapeType=Capsule, length=0.7, radius=0.505, shapeName=dyn_tassle_G_04_capsule, center=<0.7 0 0>, rotation=<rot 90 0 0>}]
 handle [{shapeType=Capsule, length=30.0228, radius=2.1516, center=<-0.1794 0.575 47.7695>}]
}
constraints [
 {bone1Name=Tassle_root, bone2Name=handle, swing1LimitDegrees=45, swing2LimitDegrees=45, twistLimitDegrees=45, twistMotion=Locked}
 {bone1Name=dyn_tassle_A_01, bone2Name=Tassle_root, swing1LimitDegrees=60, swing2LimitDegrees=30, twistLimitDegrees=45, twistMotion=Locked}
 {bone1Name=dyn_tassle_A_02, bone2Name=dyn_tassle_A_01, swing1LimitDegrees=50, swing2LimitDegrees=20, twistLimitDegrees=45, twistMotion=Locked}
 {bone1Name=dyn_tassle_A_03, bone2Name=dyn_tassle_A_02, swing1LimitDegrees=40, swing2LimitDegrees=20, twistLimitDegrees=45, twistMotion=Locked}
 {bone1Name=dyn_tassle_A_04, bone2Name=dyn_tassle_A_03, swing1LimitDegrees=35, swing2LimitDegrees=15, twistLimitDegrees=45, twistMotion=Locked}
 {bone1Name=dyn_tassle_B_01, bone2Name=Tassle_root, swing1LimitDegrees=60, swing2LimitDegrees=35, twistLimitDegrees=45, twistMotion=Locked}
 {bone1Name=dyn_tassle_B_02, bone2Name=dyn_tassle_B_01, swing1LimitDegrees=50, swing2LimitDegrees=20, twistLimitDegrees=45, twistMotion=Locked}
 {bone1Name=dyn_tassle_B_03, bone2Name=dyn_tassle_B_02, swing1LimitDegrees=40, swing2LimitDegrees=20, twistLimitDegrees=45, twistMotion=Locked}
 {bone1Name=dyn_tassle_B_04, bone2Name=dyn_tassle_B_03, swing1LimitDegrees=35, swing2LimitDegrees=15, twistLimitDegrees=45, twistMotion=Locked}
 {bone1Name=dyn_tassle_C_01, bone2Name=Tassle_root, swing1LimitDegrees=60, swing2LimitDegrees=35, twistLimitDegrees=45, twistMotion=Locked}
 {bone1Name=dyn_tassle_C_02, bone2Name=dyn_tassle_C_01, swing1LimitDegrees=50, swing2LimitDegrees=20, twistLimitDegrees=45, twistMotion=Locked}
 {bone1Name=dyn_tassle_C_03, bone2Name=dyn_tassle_C_02, swing1LimitDegrees=40, swing2LimitDegrees=20, twistLimitDegrees=45, twistMotion=Locked}
 {bone1Name=dyn_tassle_C_04, bone2Name=dyn_tassle_C_03, swing1LimitDegrees=35, swing2LimitDegrees=15, twistLimitDegrees=45, twistMotion=Locked}
 {bone1Name=dyn_tassle_D_01, bone2Name=Tassle_root, swing1LimitDegrees=60, swing2LimitDegrees=35, twistLimitDegrees=45, twistMotion=Locked}
 {bone1Name=dyn_tassle_D_02, bone2Name=dyn_tassle_D_01, swing1LimitDegrees=50, swing2LimitDegrees=25, twistLimitDegrees=45, twistMotion=Locked}
 {bone1Name=dyn_tassle_D_03, bone2Name=dyn_tassle_D_02, swing1LimitDegrees=40, swing2LimitDegrees=20, twistLimitDegrees=45, twistMotion=Locked}
 {bone1Name=dyn_tassle_D_04, bone2Name=dyn_tassle_D_03, swing1LimitDegrees=35, swing2LimitDegrees=15, twistLimitDegrees=45, twistMotion=Locked}
 {bone1Name=dyn_tassle_E_01, bone2Name=Tassle_root, swing1LimitDegrees=60, swing2LimitDegrees=35, twistLimitDegrees=45, twistMotion=Locked}
 {bone1Name=dyn_tassle_E_02, bone2Name=dyn_tassle_E_01, swing1LimitDegrees=50, swing2LimitDegrees=25, twistLimitDegrees=45, twistMotion=Locked}
 {bone1Name=dyn_tassle_E_03, bone2Name=dyn_tassle_E_02, swing1LimitDegrees=35, swing2LimitDegrees=20, twistLimitDegrees=45, twistMotion=Locked}
 {bone1Name=dyn_tassle_E_04, bone2Name=dyn_tassle_E_03, swing1LimitDegrees=35, swing2LimitDegrees=15, twistLimitDegrees=45, twistMotion=Locked}
 {bone1Name=dyn_tassle_F_01, bone2Name=Tassle_root, swing1LimitDegrees=60, swing2LimitDegrees=35, twistLimitDegrees=45, twistMotion=Locked}
 {bone1Name=dyn_tassle_F_02, bone2Name=dyn_tassle_F_01, swing1LimitDegrees=50, swing2LimitDegrees=35, twistLimitDegrees=45, twistMotion=Locked}
 {bone1Name=dyn_tassle_F_03, bone2Name=dyn_tassle_F_02, swing1LimitDegrees=30, swing2LimitDegrees=20, twistLimitDegrees=45, twistMotion=Locked}
 {bone1Name=dyn_tassle_F_04, bone2Name=dyn_tassle_F_03, swing1LimitDegrees=35, swing2LimitDegrees=15, twistLimitDegrees=45, twistMotion=Locked}
 {bone1Name=dyn_tassle_G_01, bone2Name=Tassle_root, swing1LimitDegrees=45, swing2LimitDegrees=30, twistLimitDegrees=45, twistMotion=Locked}
 {bone1Name=dyn_tassle_G_02, bone2Name=dyn_tassle_G_01, swing1LimitDegrees=40, swing2LimitDegrees=30, twistLimitDegrees=45, twistMotion=Locked}
 {bone1Name=dyn_tassle_G_03, bone2Name=dyn_tassle_G_02, swing1LimitDegrees=30, swing2LimitDegrees=20, twistLimitDegrees=45, twistMotion=Locked}
 {bone1Name=dyn_tassle_G_04, bone2Name=dyn_tassle_G_03, swing1LimitDegrees=30, swing2LimitDegrees=15, twistLimitDegrees=45, twistMotion=Locked}
 {
  bone1Name=dyn_string_01
  bone2Name=handle
  swing1LimitDegrees=20
  swing1Motion=Locked
  swing2LimitDegrees=20
  swing2Motion=Locked
  twistLimitDegrees=90
  twistMotion=Locked
 }
 {bone1Name=dyn_string_02, bone2Name=dyn_string_01, swing1LimitDegrees=10, swing2LimitDegrees=90, twistLimitDegrees=45, twistMotion=Locked}
 {bone1Name=dyn_string_03, bone2Name=dyn_string_02, swing1LimitDegrees=30, swing2LimitDegrees=45, twistLimitDegrees=45, twistMotion=Locked}
 {bone1Name=dyn_string_04, bone2Name=dyn_string_03, swing1LimitDegrees=25, swing2LimitDegrees=35, twistLimitDegrees=45, twistMotion=Locked}
 {bone1Name=dyn_string_05, bone2Name=dyn_string_04, swing1LimitDegrees=20, swing2LimitDegrees=20, twistLimitDegrees=45, twistMotion=Locked}
 {bone1Name=dyn_string_06, bone2Name=dyn_string_05, swing1LimitDegrees=30, swing2LimitDegrees=35, twistLimitDegrees=45, twistMotion=Locked}
 {bone1Name=dyn_ball, bone2Name=dyn_string_06, swing1LimitDegrees=45, swing2LimitDegrees=45, twistLimitDegrees=45, twistMotion=Locked}
]
