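# Animation Blueprint for the ForestJeanWink keytar instrument cosmetic, adding physics-driven secondary motion to its skeleton.
# The input pose runs through two chained RigidBody nodes sharing one physics asset, simulated in root-bone space with differently tuned motion limits.
# The second pass is active only outside the front end.
# The event graph lowers GravityOverride while skydiving, but both nodes show fixed gravity rather than reading that variable.
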AnimBlueprint /SparksCosmetics/Instrument/Keytar/ForestJeanWink/Meshes/ForestJeanWink_Keytar_AnimBP
target_skeleton=&/ForestJeanWink_Keytar_Skeleton
graphs [
 AnimGraph {nodes [
   n1 = RigidBody!(ComponentPose=Local To Component(LocalPose=Input Pose()), bAlphaBoolEnabled:bool, OverrideWorldGravity:Vector="0,0,-980", ExternalForce) -> (Pose) {
    alphaInputType=Bool
    bClampLinearTranslationLimitToRefPose=true
    bOverrideWorldGravity=true
    overridePhysicsAsset=&/ForestJeanWink_Keytar_Physics
    simulationSpace=BaseBoneSpace
    baseBoneRef {boneName=root}
    componentAppliedLinearAccClamp=<3000 3000 3000>
    simSpaceSettings {
     maxAngularAcceleration=1000
     maxAngularVelocity=3
     maxLinearAcceleration=300
     maxLinearVelocity=700
     worldAlpha=0.8
     externalLinearDragV {y=0.5, z=0.5}
    }
   }
   n2 = RigidBody(ComponentPose=@n1, bAlphaBoolEnabled=NOT Boolean(A=Get bIsInFrontEnd()), OverrideWorldGravity="0,0,-980") {
    alphaInputType=Bool
    bClampLinearTranslationLimitToRefPose=true
    bOverrideWorldGravity=true
    overridePhysicsAsset=&ForestJeanWink_Keytar_Physics
    simulationSpace=BaseBoneSpace
    baseBoneRef {boneName=root}
    componentAppliedLinearAccClamp=<3000 3000 3000>
    simSpaceSettings {
     maxAngularAcceleration=500
     maxAngularVelocity=3
     maxLinearAcceleration=700
     maxLinearVelocity=600
     velocityScaleZ=0.6
     worldAlpha=0.8
     externalLinearDragV {y=0.5, z=0.5}
    }
   }
   Output Pose(Result=Component To Local(ComponentPose=@n2))
  ]}
 EventGraph {nodes [
   Event BlueprintUpdateAnimation()
   > n1 = Branch(Condition=Get bIsSkyDiving())
   > Set GravityOverride(GravityOverride="0,0,-300")
   Set GravityOverride(execute=@n1.else, GravityOverride="0,0,-980")
   TryGetPawnOwner()
  ]}
]
variables [GravityOverride {type=Vector}]
